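; Gameplay cue burst notify for a HeatSnake movement projectile hitting the world, adjusting the impact's decal and physics effects per platform.
; On burst it checks for mobile, Switch or PC performance mode and, on any of these, destroys the burst decal.
; Otherwise it spawns an FN_RadialForce physics field actor at the cue location and shrinks the decal's relative scale.
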(fn OnBurst(EitherVFX-preferthisversion)
  (params (MyTarget :type Actor%20Object%20Reference) (Parameters :type Gameplay%20Cue%20Parameters%20Structure%20%28by%20ref%29) (ParticleComponents :type Array%20of%20FXSystem%20Component%20Object%20References) (AudioComponents :type Array%20of%20Audio%20Component%20Object%20References) (BurstCameraShakeInstance :type Camera%20Shake%20Base%20Object%20Reference) (BurstDecalInstance :type Decal%20Actor%20Object%20Reference))
  (bind (_normalizedmagnitude _rawmagnitude _effectcontext _matchedtagname _originaltag _aggregatedsourcetags _aggregatedtargettags _location) (Ability|GameplayCue|BreakGameplayCueParameters Parameters))
  (if (Math|Boolean|ORBoolean (Fortnite|IsMobilePlatform) (Class|PlatformHelpersLibrary|IsPcPerfMode MyTarget) (TouchPlatform|IsMobileGame MyTarget) (TouchPlatform|IsNintendoSwitchorNintendoSwitchPreview))
    (Actor|DestroyActor BurstDecalInstance)
    (else
      (Game|SpawnActorfromClass 0 "/Game/Effects/Niagara/Blueprints/PhsyicsFields/FN_RadialForce.FN_RadialForce_C" (Math|Conversions|ToTransform(Vector) _location) "Undefined" "MultiplyWithRoot" 0 3.0 0.75 4.8e+02 1.3 0.05 0)
      (Transformation|SetActorRelativeScale3D BurstDecalInstance (Math|Conversions|ToVector(Float) 0.6))))
  (return))
