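; Burst gameplay cue for the third HeatSnake melee hit, adding a heavy impact sound and an oriented particle effect.
; OnBurst runs the parent burst, reads the hit result from the effect context, and rotates a spawned particle component along the hit normal.
; PlayHeavyHit plays the ChainWhip heavy-hit event through the Sound Library at the hit location, flagging audio components for whether target or instigator is locally viewed.
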(fn OnBurst(EitherVFX-preferthisversion)
  (params (MyTarget :type Actor%20Object%20Reference) (Parameters :type Gameplay%20Cue%20Parameters%20Structure%20%28by%20ref%29) (ParticleComponents :type Array%20of%20FXSystem%20Component%20Object%20References) (AudioComponents :type Array%20of%20Audio%20Component%20Object%20References) (BurstCameraShakeInstance :type Camera%20Shake%20Base%20Object%20Reference) (BurstDecalInstance :type Decal%20Actor%20Object%20Reference))
  (bind (_normalizedmagnitude _rawmagnitude _effectcontext) (Ability|GameplayCue|BreakGameplayCueParameters Parameters))
  (bind (_returnvalue_bblockinghit _returnvalue_binitialoverlap _returnvalue_time _returnvalue_distance _returnvalue_location _returnvalue_impactpoint _returnvalue_normal _returnvalue_impactnormal _returnvalue_physmat _returnvalue_hitactor) (Collision|BreakHitResult (Ability|EffectContext|GetHitResult _effectcontext)))
  (|CallParentFunction MyTarget Parameters ParticleComponents AudioComponents BurstCameraShakeInstance BurstDecalInstance)
  (bind _asfort_pawn (Utilities|Casting|CastToFortPawn MyTarget)
    (:then
      _)
    (:CastFailed
      _))
  (bind _asfort_pawn_2 (Utilities|Casting|CastToFortPawn _returnvalue_hitactor))
  (CallFunction|PlayHeavyHit _asfort_pawn_2 _asfort_pawn (Variables|ChainWhip|GetHeavyHitEvent) _returnvalue_location)
  (Transformation|SetRelativeRotation (Utilities|Array|Get(acopy) ParticleComponents) (Math|Conversions|RotationFromXVector _returnvalue_normal) false false)
  (return))

(fn PlayHeavyHit
  (params (Target :type Fort%20Pawn%20Object%20Reference) (Instigator :type Fort%20Pawn%20Object%20Reference) (EventName :type Gameplay%20Tag%20Structure) (Location :type Vector :default "0, 0, 0"))
  (bind _returnvalue (WorldSubsystems|GetSoundLibrarySubsystem Instigator))
  ;[0.007682 0.1219 0.1354] PlaySound: Play the desired event name sound at the provided location.
  (Utilities|IsValid Target
    (:Is%20Valid
      (Utilities|IsValid _returnvalue
        (:Is%20Valid
          (bind _returnvalue_2 (SoundLibrary|PlaySoundSimple _returnvalue Instigator (Utilities|Struct|MakeSimpleSoundSettings EventName false false true true 1.0 1.0 0.0 0 0 "None" 0 "KeepWorldPosition" Location) :!class /Script/SoundLibrary.SoundLibrarySubsystem))
          (for _array_element (.AudioComponents (Utilities|Struct|BreakSoundLibraryPlayResult _returnvalue_2))
            (Audio|Parameter|SetBooleanParameter _array_element "istarget" (Pawn|IsLocallyViewed (Variables|GetTarget)) :!class /Script/AudioExtensions.AudioParameterControllerInterface)
            (Audio|Parameter|SetBooleanParameter _array_element "isinstigator" (Pawn|IsLocallyViewed (Variables|GetInstigator)) :!class /Script/AudioExtensions.AudioParameterControllerInterface)))))))
